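# Sound cue for the Venus pickaxe impact in the cosmetics content, routed through the impacts sound class.
# A mixer layers a random pick of three Venus impact waves with a pitched-down random pick of four shared player flesh impact waves.
# A short fixed delay precedes the mix.
# Shared pickaxe attenuation, property sheet and melee weapon concurrency assets govern playback, and the cue primes on load.
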SoundCue /BRCosmetics/Sounds/PickAxes/Venus/Pickaxe_Venus_Impact_Cue
root=Delay_1
nodes [
 {id=Delay_1, children [Mixer_0], properties {delayMax=0.1, delayMin=0.1}}
 {id=Mixer_0, children [Random_0,Modulator_0], properties {inputVolume [0.85,0.8]}}
 {id=Modulator_0, children [Random_1], properties {pitchMax=0.85, pitchMin=0.85}}
 {id=Random_0, children [WavePlayer_0,WavePlayer_1,WavePlayer_2], properties {weights [1,1,1]}}
 {id=Random_1, children [WavePlayer_4,WavePlayer_5,WavePlayer_6,WavePlayer_7], properties {weights [1,1,1,1]}}
 {id=WavePlayer_0, properties {soundWaveAssetPtr=&/Pickaxe_Venus_Impact_01}}
 {id=WavePlayer_1, properties {soundWaveAssetPtr=&/Pickaxe_Venus_Impact_02}}
 {id=WavePlayer_2, properties {soundWaveAssetPtr=&/Pickaxe_Venus_Impact_03}}
 {id=WavePlayer_4, properties {soundWaveAssetPtr=/Game/Sounds/Fort_Impact_Sounds/Flesh/new_player_impact_nonspat_A_01n}}
 {id=WavePlayer_5, properties {soundWaveAssetPtr=/Game/Sounds/Fort_Impact_Sounds/Flesh/new_player_impact_nonspat_A_02n}}
 {id=WavePlayer_6, properties {soundWaveAssetPtr=/Game/Sounds/Fort_Impact_Sounds/Flesh/new_player_impact_nonspat_A_03n}}
 {id=WavePlayer_7, properties {soundWaveAssetPtr=/Game/Sounds/Fort_Impact_Sounds/Flesh/new_player_impact_nonspat_A_04n}}
]
properties {
 attenuationSettings=/Game/Sounds/Attenuation/Fort_Pickaxe_Impact_Att
 audioPropertiesSheet=/BRCosmetics/Sounds/Settings/PropertySheets/Pickaxe/APS_Cosmetic_Pickaxe_Impact
 bPrimeOnLoad=true
 maxDistance=4300
 pitchMultiplier=0.95
 soundClassObject=&Fort_Impacts
 volumeMultiplier=0.5625
 attenuationOverrides {
  bAttenuateWithLPF=true
  distanceAlgorithm=Logarithmic
  falloffDistance=800
  lPFRadiusMax=800
  lPFRadiusMin=400
  attenuationShapeExtents {x=300}
 }
 concurrencyOverrides {bLimitToOwner=true, maxCount=2, resolutionRule=StopOldest}
 concurrencySet [&Player_Melee_Weapon_Conc]
}
